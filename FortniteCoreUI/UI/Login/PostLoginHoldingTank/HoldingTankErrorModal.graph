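# Post-login holding tank error modal, built on FortHoldingTankErrorModal, with inherited title, scrollable body, error code, retry countdown and Confirm/Decline button widgets.
# On Construct, UpdateButtonsOrientation maps device orientation through a macro to stack the buttons vertically or horizontally, adjusting alignment and sizing.
# It also binds to PlatformGameInstance's orientation-change delegate, unbinding on Destruct.
# The OnOrientationChanged handler sets only stack orientation, leaving alignment and sizing unchanged.
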WidgetBlueprint /FortniteCoreUI/UI/Login/PostLoginHoldingTank/HoldingTankErrorModal
graphs [
 DeviceOrienationToWidgetOrientation {nodes [Outputs!(ReturnValue:EOrientation Enum=Select!(
     Unknown:EOrientation Enum=Orient_Horizontal
     Portrait:EOrientation Enum=Orient_Vertical
     PortraitUpsideDown:EOrientation Enum=Orient_Vertical
     LandscapeLeft:EOrientation Enum=Orient_Horizontal
     LandscapeRight:EOrientation Enum=Orient_Horizontal
     FaceUp:EOrientation Enum=Orient_Horizontal
     FaceDown:EOrientation Enum=Orient_Horizontal
     PortraitSensor:EOrientation Enum=Orient_Vertical
     LandscapeSensor:EOrientation Enum=Orient_Horizontal
     FullSensor:EOrientation Enum=Orient_Horizontal
     Index:EScreenOrientation Enum=Inputs!() -> (Index:EScreenOrientation Enum=Unknown)
    ) -> (ReturnValue:EOrientation Enum))]}
 EventGraph {nodes [
   n1 = OnOrientationChanged!() -> (OutputDelegate, then, inScreenOrientation:EScreenOrientation Enum=Unknown) {is=title K2Node_CustomEvent}
   > Set Orientation!(
    execute
    Orientation:EOrientation Enum=Device Orienation to Widget Orientation!(Index:EScreenOrientation Enum=@n1.inScreenOrientation) -> (ReturnValue:EOrientation Enum=Orient_Horizontal) {is=title K2Node_MacroInstance}
    self:Stack Box Object Reference=Get Buttons_StackBox() -> (Buttons_StackBox:Stack Box Object Reference)
   ) -> (then, Output_Get:EOrientation Enum=Orient_Horizontal)
   Event Construct()
   > n2 = Sequence()
   n3 = Cast To PlatformGameInstance!(execute=@n2.then_1, Object=GetGameInstance!(self:Widget Object Reference) -> (ReturnValue:Game Instance Object Reference)) -> (then, CastFailed, AsPlatform Game Instance:Platform Game Instance Object Reference) {is=title K2Node_DynamicCast}
   > n4 = Set PlatformGameInstance(PlatformGameInstance:Platform Game Instance Object Reference=@n3) -> (Output_Get:Platform Game Instance Object Reference)
   > Bind Event to Application Received Screen Orientation Changed Notification Delegate(self:Platform Game Instance Object Reference=@n4, Delegate=Create Event()) {is=title K2Node_AddDelegate}
   UpdateButtonsOrientation!(execute=@n2.then_0, self, ScreenOrientation:EScreenOrientation Enum=GetDeviceOrientation!() -> (ReturnValue:EScreenOrientation Enum=Unknown)) -> (then)
   Event Destruct()
   n5 = Get PlatformGameInstance() -> (PlatformGameInstance:Platform Game Instance Object Reference)
   > Is Valid(InputObject=@n5)
   > Unbind Event from Application Received Screen Orientation Changed Notification Delegate(self:Platform Game Instance Object Reference=@n5, Delegate=Create Event()) {is=title K2Node_RemoveDelegate}
  ]}
 UpdateButtonsOrientation {nodes [
   n1 = UpdateButtonsOrientation!() -> (then, ScreenOrientation:EScreenOrientation Enum=Unknown) {is=title K2Node_FunctionEntry}
   n2 = Device Orienation to Widget Orientation(Index=@n1) -> (ReturnValue=Orient_Horizontal)
   n3 = Get Buttons_StackBox()
   > Set Orientation(Orientation=@n2, self=@n3) -> (Output_Get=Orient_Horizontal)
   > SetHorizontalAlignment(
    self:Vertical Box Slot Object Reference=SlotAsVerticalBoxSlot(Widget=@n3)
    InHorizontalAlignment=Select!(Orient_Horizontal:EHorizontalAlignment Enum=HAlign_Left, Orient_Vertical:EHorizontalAlignment Enum=HAlign_Fill, Index:EOrientation Enum=@n2) -> (ReturnValue:EHorizontalAlignment Enum)
   )
   > Set Size!(
    execute
    Size:Slate Child Size Structure=Select(
     Orient_Horizontal:Slate Child Size Structure=Make Slate Child Size!(Value:float=1, SizeRule:ESlateSizeRule Enum=Automatic) -> (SlateChildSize:Slate Child Size Structure) {is=title K2Node_MakeStruct}
     Orient_Vertical:Slate Child Size Structure=Make Slate Child Size(Value=1, SizeRule=Fill)
     Index=@n2
    ) -> (ReturnValue:Slate Child Size Structure)
    self:Stack Box Slot Object Reference=SlotAsStackBoxSlot!(Widget:Widget Object Reference=Get ConfirmButton() -> (ConfirmButton:Button Object Reference)) -> (ReturnValue:Stack Box Slot Object Reference)
   ) -> (then, Output_Get:Slate Child Size Structure)
   > Set Size(
    Size=Select(
     Orient_Horizontal=Make Slate Child Size(Value=1, SizeRule=Automatic)
     Orient_Vertical=Make Slate Child Size(Value=1, SizeRule=Fill)
     Index=@n2
    )
    self=SlotAsStackBoxSlot(Widget=Get DeclineButton() -> (DeclineButton:Button Object Reference))
   )
  ]}
]
variables [PlatformGameInstance {type=PlatformGameInstance, default=null}]
widgets {
 info {inheritedWidgetCount=8, parentClass=/Script/FortniteUI.FortHoldingTankErrorModal, rootWidgetClass=&SafeZone, widgetCount=20}
 widgets [
  SafeZone {widgetClassPath=&SafeZone, widget=SafeZone, properties {padLeft=false, padRight=false}}
  Border_101 {
   widgetClassPath=&Border
   parent=SafeZone
   widget=Border_101
   properties {horizontalAlignment=HAlign_Center, verticalAlignment=VAlign_Center, brushColor=<rgba 0 0 0 0>, padding=<ltrb 50 50 50 50>}
   slot {class=SafeZoneSlot}
  }
  SizeBox_82 {
   widgetClassPath=&SizeBox
   parent=Border_101
   widget=SizeBox_82
   properties {
    bOverride_MaxDesiredWidth=true
    bOverride_MinDesiredHeight=true
    bOverride_MinDesiredWidth=true
    maxDesiredWidth=1500
    minDesiredHeight=700
    minDesiredWidth=800
   }
   slot {class=BorderSlot, properties {horizontalAlignment=HAlign_Center, verticalAlignment=VAlign_Center, padding=<ltrb 50 50 50 50>}}
  }
  Overlay_58 {widgetClassPath=&Overlay, parent=SizeBox_82, widget=Overlay_58, slot {class=SizeBoxSlot}}
  Image_95 {
   widgetClassPath=&Image
   parent=Overlay_58
   widget=Image_95
   properties {
    brush {
     bIsDynamicallyLoaded=false
     drawAs=RoundedBox
     imageType=NoImage
     mirroring=NoMirror
     resourceName=None
     resourceObject=null
     tiling=NoTile
     imageSize=<32 32>
     margin=<ltrb 0 0 0 0>
     outlineSettings {
      bUseBrushTransparency=false
      roundingType=FixedRadius
      width=0
      color {colorUseRule=UseColor_Specified, specifiedColor=<rgba 0 0 0 0>}
      cornerRadii=<20 20 20 20>
     }
     tintColor {colorUseRule=UseColor_Specified, specifiedColor=<rgba 1 1 1 1>}
     uVRegion {bIsValid=false, max=<0 0>, min=<0 0>}
    }
    colorAndOpacity=<rgba 0.0075 0.0123 0.0222 1>
   }
   slot {class=OverlaySlot, properties {horizontalAlignment=HAlign_Fill, verticalAlignment=VAlign_Fill}}
  }
  VerticalBox {
   widgetClassPath=&VerticalBox
   parent=Overlay_58
   widget=VerticalBox
   slot {class=OverlaySlot, properties {horizontalAlignment=HAlign_Fill, verticalAlignment=VAlign_Fill, padding=<ltrb 0 0 0 20>}}
  }
  Title_SizeBox {
   widgetClassPath=&SizeBox
   parent=VerticalBox
   widget=Title_SizeBox
   properties {bOverride_MinDesiredHeight=true, minDesiredHeight=100}
   slot {class=VerticalBoxSlot}
  }
  Border_254 {
   widgetClassPath=&Border
   parent=Title_SizeBox
   widget=Border_254
   properties {
    verticalAlignment=VAlign_Center
    background {
     bIsDynamicallyLoaded=false
     drawAs=RoundedBox
     imageType=NoImage
     mirroring=NoMirror
     resourceName=None
     resourceObject=null
     tiling=NoTile
     imageSize=<32 32>
     margin=<ltrb 0 0 0 0>
     outlineSettings {
      bUseBrushTransparency=false
      roundingType=FixedRadius
      width=0
      color {colorUseRule=UseColor_Specified, specifiedColor=<rgba 0 0 0 0>}
      cornerRadii=<20 20 0 0>
     }
     tintColor {colorUseRule=UseColor_Specified, specifiedColor=<rgba 0.0048 0.0065 0.0103 1>}
     uVRegion {bIsValid=false, max=<0 0>, min=<0 0>}
    }
    brushColor=<rgba 0.2918 0.314 0.3419 1>
    padding=<ltrb 100 35 100 35>
   }
   slot {class=SizeBoxSlot}
  }
  TitleTextBlock {
   widgetClassPath=&CommonTextBlock
   bInherited=true
   bIsVariable=true
   parent=Border_254
   widget=TitleTextBlock
   properties {
    autoWrapText=true
    justification=Center
    text=Title
    colorAndOpacity {colorUseRule=UseColor_Specified, specifiedColor=<rgba 0.314 0.3372 0.3663 1>}
    font {
     bForceMonospaced=false
     bMaterialIsStencil=false
     fontMaterial=null
     fontObject=&HeadingNow
     letterSpacing=0
     monospacedWidth=1
     size=45.75
     skewAmount=0
     typefaceFontName=76 Bold
     outlineSettings {
      bApplyOutlineToDropShadows=false
      bMiteredCorners=false
      bSeparateFillAlpha=false
      outlineMaterial=null
      outlineSize=0
      outlineColor=<rgba 0 0 0 1>
     }
    }
    shadowOffset=<0 0>
   }
   slot {class=BorderSlot, properties {verticalAlignment=VAlign_Center, padding=<ltrb 100 35 100 35>}}
  }
  Body_ScrollBox {
   widgetClassPath=&ScrollBox
   parent=VerticalBox
   widget=Body_ScrollBox
   slot {class=VerticalBoxSlot, properties {padding=<ltrb 65 60 65 0>, size {sizeRule=Fill, value=1}}}
  }
  Overlay_142 {widgetClassPath=&Overlay, parent=Body_ScrollBox, widget=Overlay_142, slot {class=ScrollBoxSlot}}
  BodyTextBlock {
   widgetClassPath=&CommonTextBlock
   bInherited=true
   bIsVariable=true
   parent=Overlay_142
   widget=BodyTextBlock
   properties {
    autoWrapText=true
    lineHeightPercentage=1.3
    text=Text Block
    colorAndOpacity {colorUseRule=UseColor_Specified, specifiedColor=<rgba 0.5776 0.5841 0.5906 1>}
    font {
     bForceMonospaced=false
     bMaterialIsStencil=false
     fontMaterial=null
     fontObject=&HeadingNow
     letterSpacing=0
     monospacedWidth=1
     size=33.75
     skewAmount=0
     typefaceFontName=74 Regular
     outlineSettings {
      bApplyOutlineToDropShadows=false
      bMiteredCorners=false
      bSeparateFillAlpha=false
      outlineMaterial=null
      outlineSize=0
      outlineColor=<rgba 0 0 0 1>
     }
    }
    shadowOffset=<0 0>
   }
   slot {class=OverlaySlot, properties {horizontalAlignment=HAlign_Fill, verticalAlignment=VAlign_Fill}}
  }
  ErrorCodeTextBlock {
   widgetClassPath=&CommonTextBlock
   bInherited=true
   bIsVariable=true
   parent=VerticalBox
   widget=ErrorCodeTextBlock
   properties {
    text=ESP-CON-000
    colorAndOpacity {colorUseRule=UseColor_Specified, specifiedColor=<rgba 0.3419 0.3515 0.3663 1>}
    font {
     bForceMonospaced=false
     bMaterialIsStencil=false
     fontMaterial=null
     fontObject=&HeadingNow
     letterSpacing=0
     monospacedWidth=1
     size=27
     skewAmount=0
     typefaceFontName=74 Regular
     outlineSettings {
      bApplyOutlineToDropShadows=false
      bMiteredCorners=false
      bSeparateFillAlpha=false
      outlineMaterial=null
      outlineSize=0
      outlineColor=<rgba 0 0 0 1>
     }
    }
    shadowOffset=<0 0>
   }
   slot {class=VerticalBoxSlot, properties {padding=<ltrb 60 10 60 10>}}
  }
  TimerTextBlock {
   widgetClassPath=&CommonTextBlock
   bInherited=true
   bIsVariable=true
   parent=VerticalBox
   widget=TimerTextBlock
   properties {
    text="Retrying in {SecondsLeft}"
    colorAndOpacity {colorUseRule=UseColor_Specified, specifiedColor=<rgba 0.3419 0.3515 0.3663 1>}
    font {
     bForceMonospaced=false
     bMaterialIsStencil=false
     fontMaterial=null
     fontObject=&HeadingNow
     letterSpacing=0
     monospacedWidth=1
     size=27
     skewAmount=0
     typefaceFontName=75 Medium
     outlineSettings {
      bApplyOutlineToDropShadows=false
      bMiteredCorners=false
      bSeparateFillAlpha=false
      outlineMaterial=null
      outlineSize=0
      outlineColor=<rgba 0 0 0 1>
     }
    }
    shadowOffset=<0 0>
   }
   slot {class=VerticalBoxSlot, properties {padding=<ltrb 60 10 60 10>}}
  }
  Spacer_137 {widgetClassPath=&Spacer, parent=VerticalBox, widget=Spacer_137, properties {size=<10 10>}, slot {class=VerticalBoxSlot}}
  Buttons_StackBox {
   widgetClassPath=&StackBox
   bIsVariable=true
   parent=VerticalBox
   widget=Buttons_StackBox
   properties {orientation=Orient_Vertical}
   slot {class=VerticalBoxSlot, properties {padding=<ltrb 50 0 50 0>}}
  }
  ConfirmButton {
   widgetClassPath=&Button
   bInherited=true
   parent=Buttons_StackBox
   widget=ConfirmButton
   properties {
    backgroundColor=<rgba 0.9301 1 0.0103 1>
    colorAndOpacity=<rgba 0 0 0 1>
    widgetStyle {
     clickedSlateSound {resourceObject=null}
     disabled {
      bIsDynamicallyLoaded=false
      drawAs=NoDrawType
      imageType=NoImage
      mirroring=NoMirror
      resourceName=None
      resourceObject=null
      tiling=NoTile
      imageSize=<0 0>
      margin=<ltrb 0 0 0 0>
      outlineSettings {
       bUseBrushTransparency=false
       roundingType=HalfHeightRadius
       width=0
       color {colorUseRule=UseColor_Specified, specifiedColor=<rgba 0 0 0 0>}
       cornerRadii=<0 0 0 0>
      }
      tintColor {colorUseRule=UseColor_Specified, specifiedColor=<rgba 1 1 1 1>}
      uVRegion {bIsValid=false, max=<0 0>, min=<0 0>}
     }
     disabledForeground {colorUseRule=UseColor_Specified, specifiedColor=<rgba 0.5271 0.5271 0.5271 1>}
     hovered {
      bIsDynamicallyLoaded=false
      drawAs=RoundedBox
      imageType=NoImage
      mirroring=NoMirror
      resourceName=None
      resourceObject=null
      tiling=NoTile
      imageSize=<32 32>
      margin=<ltrb 0 0 0 0>
      outlineSettings {
       bUseBrushTransparency=true
       roundingType=FixedRadius
       width=0
       color {colorUseRule=UseColor_Specified, specifiedColor=<rgba 0.7243 0.7243 0.7243 1>}
       cornerRadii=<20 20 20 20>
      }
      tintColor {colorUseRule=UseColor_Specified, specifiedColor=<rgba 1 1 0.7243 1>}
      uVRegion {bIsValid=false, max=<0 0>, min=<0 0>}
     }
     hoveredForeground {colorUseRule=UseColor_Specified, specifiedColor=<rgba 0 1 1 1>}
     hoveredSlateSound {resourceObject=null}
     normal {
      bIsDynamicallyLoaded=false
      drawAs=RoundedBox
      imageType=NoImage
      mirroring=NoMirror
      resourceName=None
      resourceObject=null
      tiling=NoTile
      imageSize=<32 32>
      margin=<ltrb 0 0 0 0>
      outlineSettings {
       bUseBrushTransparency=true
       roundingType=FixedRadius
       width=0
       color {colorUseRule=UseColor_Specified, specifiedColor=<rgba 0.6951 0.6951 0.6951 1>}
       cornerRadii=<20 20 20 20>
      }
      tintColor {colorUseRule=UseColor_Specified, specifiedColor=<rgba 1 1 1 1>}
      uVRegion {bIsValid=false, max=<0 0>, min=<0 0>}
     }
     normalForeground {colorUseRule=UseColor_Specified, specifiedColor=<rgba 0 1 1 1>}
     normalPadding=<ltrb 12 1.5 12 1.5>
     pressed {
      bIsDynamicallyLoaded=false
      drawAs=RoundedBox
      imageType=NoImage
      mirroring=NoMirror
      resourceName=None
      resourceObject=null
      tiling=NoTile
      imageSize=<32 32>
      margin=<ltrb 0 0 0 0>
      outlineSettings {
       bUseBrushTransparency=true
       roundingType=FixedRadius
       width=0
       color {colorUseRule=UseColor_Specified, specifiedColor=<rgba 0.7243 0.7243 0.7243 1>}
       cornerRadii=<20 20 20 20>
      }
      tintColor {colorUseRule=UseColor_Specified, specifiedColor=<rgba 0.3843 0.3843 0.3843 1>}
      uVRegion {bIsValid=false, max=<0 0>, min=<0 0>}
     }
     pressedForeground {colorUseRule=UseColor_Specified, specifiedColor=<rgba 0 1 1 1>}
     pressedPadding=<ltrb 12 1.5 12 1.5>
     pressedSlateSound {resourceObject=null}
    }
   }
   slot {class=StackBoxSlot, properties {padding=<ltrb 10 15 10 15>, size {sizeRule=Fill, value=1}}}
  }
  ConfirmTextBlock {
   widgetClassPath=&CommonTextBlock
   bInherited=true
   bIsVariable=true
   parent=ConfirmButton
   widget=ConfirmTextBlock
   properties {
    text=CONFIRM
    font {
     bForceMonospaced=false
     bMaterialIsStencil=false
     fontMaterial=null
     fontObject=&HeadingNow
     letterSpacing=0
     monospacedWidth=1
     size=45
     skewAmount=0
     typefaceFontName=76 Bold
     outlineSettings {
      bApplyOutlineToDropShadows=false
      bMiteredCorners=false
      bSeparateFillAlpha=false
      outlineMaterial=null
      outlineSize=0
      outlineColor=<rgba 0 0 0 1>
     }
    }
    shadowOffset=<0 0>
   }
   slot {class=ButtonSlot, properties {padding=<ltrb 10 10 10 10>}}
  }
  DeclineButton {
   widgetClassPath=&Button
   bInherited=true
   parent=Buttons_StackBox
   widget=DeclineButton
   properties {
    backgroundColor=<rgba 0.3613 0.4678 0.8469 1>
    widgetStyle {
     clickedSlateSound {resourceObject=null}
     disabled {
      bIsDynamicallyLoaded=false
      drawAs=NoDrawType
      imageType=NoImage
      mirroring=NoMirror
      resourceName=None
      resourceObject=null
      tiling=NoTile
      imageSize=<0 0>
      margin=<ltrb 0 0 0 0>
      outlineSettings {
       bUseBrushTransparency=false
       roundingType=HalfHeightRadius
       width=0
       color {colorUseRule=UseColor_Specified, specifiedColor=<rgba 0 0 0 0>}
       cornerRadii=<0 0 0 0>
      }
      tintColor {colorUseRule=UseColor_Specified, specifiedColor=<rgba 1 1 1 1>}
      uVRegion {bIsValid=false, max=<0 0>, min=<0 0>}
     }
     disabledForeground {colorUseRule=UseColor_Specified, specifiedColor=<rgba 0.5271 0.5271 0.5271 1>}
     hovered {
      bIsDynamicallyLoaded=false
      drawAs=RoundedBox
      imageType=NoImage
      mirroring=NoMirror
      resourceName=None
      resourceObject=null
      tiling=NoTile
      imageSize=<32 32>
      margin=<ltrb 0 0 0 0>
      outlineSettings {
       bUseBrushTransparency=true
       roundingType=FixedRadius
       width=0
       color {colorUseRule=UseColor_Specified, specifiedColor=<rgba 0.7243 0.7243 0.7243 1>}
       cornerRadii=<20 20 20 20>
      }
      tintColor {colorUseRule=UseColor_Specified, specifiedColor=<rgba 0.7243 0.7243 0.7243 1>}
      uVRegion {bIsValid=false, max=<0 0>, min=<0 0>}
     }
     hoveredForeground {colorUseRule=UseColor_Specified, specifiedColor=<rgba 0 1 1 1>}
     hoveredSlateSound {resourceObject=null}
     normal {
      bIsDynamicallyLoaded=false
      drawAs=RoundedBox
      imageType=NoImage
      mirroring=NoMirror
      resourceName=None
      resourceObject=null
      tiling=NoTile
      imageSize=<32 32>
      margin=<ltrb 0 0 0 0>
      outlineSettings {
       bUseBrushTransparency=true
       roundingType=FixedRadius
       width=0
       color {colorUseRule=UseColor_Specified, specifiedColor=<rgba 1 1 1 1>}
       cornerRadii=<20 20 20 20>
      }
      tintColor {colorUseRule=UseColor_Specified, specifiedColor=<rgba 1 1 1 1>}
      uVRegion {bIsValid=false, max=<0 0>, min=<0 0>}
     }
     normalForeground {colorUseRule=UseColor_Specified, specifiedColor=<rgba 0 1 1 1>}
     normalPadding=<ltrb 12 1.5 12 1.5>
     pressed {
      bIsDynamicallyLoaded=false
      drawAs=RoundedBox
      imageType=NoImage
      mirroring=NoMirror
      resourceName=None
      resourceObject=null
      tiling=NoTile
      imageSize=<32 32>
      margin=<ltrb 0 0 0 0>
      outlineSettings {
       bUseBrushTransparency=true
       roundingType=FixedRadius
       width=0
       color {colorUseRule=UseColor_Specified, specifiedColor=<rgba 0.7243 0.7243 0.7243 1>}
       cornerRadii=<4 4 4 4>
      }
      tintColor {colorUseRule=UseColor_Specified, specifiedColor=<rgba 0.3843 0.3843 0.3843 1>}
      uVRegion {bIsValid=false, max=<0 0>, min=<0 0>}
     }
     pressedForeground {colorUseRule=UseColor_Specified, specifiedColor=<rgba 0 1 1 1>}
     pressedPadding=<ltrb 12 1.5 12 1.5>
     pressedSlateSound {resourceObject=null}
    }
   }
   slot {class=StackBoxSlot, properties {padding=<ltrb 10 15 10 15>}}
  }
  DeclineTextBlock {
   widgetClassPath=&CommonTextBlock
   bInherited=true
   bIsVariable=true
   parent=DeclineButton
   widget=DeclineTextBlock
   properties {
    text=DECLINE
    colorAndOpacity {colorUseRule=UseColor_Specified, specifiedColor=<rgba 0 0 0 1>}
    font {
     bForceMonospaced=false
     bMaterialIsStencil=false
     fontMaterial=null
     fontObject=&HeadingNow
     letterSpacing=0
     monospacedWidth=1
     size=45
     skewAmount=0
     typefaceFontName=76 Bold
     outlineSettings {
      bApplyOutlineToDropShadows=false
      bMiteredCorners=false
      bSeparateFillAlpha=false
      outlineMaterial=null
      outlineSize=0
      outlineColor=<rgba 0 0 0 1>
     }
    }
    shadowOffset=<0 0>
   }
   slot {class=ButtonSlot, properties {padding=<ltrb 10 10 10 10>}}
  }
 ]
}
