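# Modal widget for status messages in the post-login holding tank, derived from FortHoldingTankStatusModal.
# A size-constrained rounded panel stacks a red title bar with alert icon, scrollable body text and an action button around inherited widget bindings.
# On Construct, AdjustTitleVisiblity (sic) only changes the title row's border-slot padding, removing bottom padding when IsAlert is false and TitleText is empty.
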WidgetBlueprint /FortniteCoreUI/UI/Login/PostLoginHoldingTank/HoldingTankStatusModal
graphs [
 AdjustTitleVisiblity {nodes [
   AdjustTitleVisiblity() {is=title K2Node_FunctionEntry}
   n1 = SlotAsBorderSlot!(Widget:Widget Object Reference=Get Title_HorizontalBox() -> (Title_HorizontalBox:Horizontal Box Object Reference)) -> (ReturnValue:Border Slot Object Reference)
   > Is Valid(InputObject=@n1)
   n2 = NOT Boolean(A=Get IsAlert() -> (IsAlert:bool))
   n3 = TextIsEmpty(InText=Get TitleText() -> (TitleText:Text))
   > SetPadding!(
    execute
    self:Border Slot Object Reference=@n1
    InPadding:Margin Structure=Select(
     Option 0:Margin Structure=Make Margin(Left=50, Top=20, Right=50, Bottom=14)
     Option 1:Margin Structure=Make Margin(Left=50, Top=20, Right=50, Bottom=0)
     Index=AND Boolean(A=@n2, B=@n3)
    ) -> (ReturnValue:Margin Structure)
   ) -> (then)
  ]}
 EventGraph {nodes [
   Event Construct()
   > AdjustTitleVisiblity() {is=title K2Node_CallFunction}
  ]}
]
variables []
widgets {
 info {inheritedWidgetCount=5, parentClass=/Script/FortniteUI.FortHoldingTankStatusModal, rootWidgetClass=&SafeZone, widgetCount=16}
 widgets [
  SafeZone {widgetClassPath=&SafeZone, widget=SafeZone, properties {padLeft=false, padRight=false}}
  Border_101 {
   widgetClassPath=&Border
   parent=SafeZone
   widget=Border_101
   properties {horizontalAlignment=HAlign_Center, verticalAlignment=VAlign_Center, brushColor=<rgba 0 0 0 0>, padding=<ltrb 40 100 40 100>}
   slot {class=SafeZoneSlot}
  }
  SizeBox_82 {
   widgetClassPath=&SizeBox
   parent=Border_101
   widget=SizeBox_82
   properties {
    bOverride_MaxDesiredWidth=true
    bOverride_MinDesiredHeight=true
    bOverride_MinDesiredWidth=true
    maxDesiredWidth=1200
    minDesiredHeight=300
    minDesiredWidth=600
   }
   slot {class=BorderSlot, properties {horizontalAlignment=HAlign_Center, verticalAlignment=VAlign_Center, padding=<ltrb 40 100 40 100>}}
  }
  Overlay_58 {widgetClassPath=&Overlay, parent=SizeBox_82, widget=Overlay_58, slot {class=SizeBoxSlot}}
  Image_95 {
   widgetClassPath=&Image
   parent=Overlay_58
   widget=Image_95
   properties {brush {
     bIsDynamicallyLoaded=false
     drawAs=RoundedBox
     imageType=NoImage
     mirroring=NoMirror
     resourceName=None
     resourceObject=null
     tiling=NoTile
     imageSize=<32 32>
     margin=<ltrb 0 0 0 0>
     outlineSettings {
      bUseBrushTransparency=false
      roundingType=FixedRadius
      width=0
      color {colorUseRule=UseColor_Specified, specifiedColor=<rgba 0 0 0 0>}
      cornerRadii=<20 20 20 20>
     }
     tintColor {colorUseRule=UseColor_Specified, specifiedColor=<rgba 0.0123 0.0103 0.0137 1>}
     uVRegion {bIsValid=false, max=<0 0>, min=<0 0>}
    }}
   slot {class=OverlaySlot, properties {horizontalAlignment=HAlign_Fill, verticalAlignment=VAlign_Fill}}
  }
  VerticalBox {
   widgetClassPath=&VerticalBox
   parent=Overlay_58
   widget=VerticalBox
   slot {class=OverlaySlot, properties {horizontalAlignment=HAlign_Fill, verticalAlignment=VAlign_Fill}}
  }
  Border_254 {
   widgetClassPath=&Border
   parent=VerticalBox
   widget=Border_254
   properties {
    horizontalAlignment=HAlign_Center
    verticalAlignment=VAlign_Center
    background {
     bIsDynamicallyLoaded=false
     drawAs=RoundedBox
     imageType=NoImage
     mirroring=NoMirror
     resourceName=None
     resourceObject=null
     tiling=NoTile
     imageSize=<32 32>
     margin=<ltrb 0 0 0 0>
     outlineSettings {
      bUseBrushTransparency=false
      roundingType=FixedRadius
      width=0
      color {colorUseRule=UseColor_Specified, specifiedColor=<rgba 0 0 0 0>}
      cornerRadii=<20 20 0 0>
     }
     tintColor {colorUseRule=UseColor_Specified, specifiedColor=<rgba 0.7682 0.011 0.1274 1>}
     uVRegion {bIsValid=false, max=<0 0>, min=<0 0>}
    }
    padding=<ltrb 50 20 50 14>
   }
   slot {class=VerticalBoxSlot}
  }
  Title_HorizontalBox {
   widgetClassPath=&HorizontalBox
   bIsVariable=true
   parent=Border_254
   widget=Title_HorizontalBox
   slot {class=BorderSlot, properties {horizontalAlignment=HAlign_Center, verticalAlignment=VAlign_Center, padding=<ltrb 50 20 50 14>}}
  }
  AlertIcon {
   widgetClassPath=&Image
   bInherited=true
   bIsVariable=true
   parent=Title_HorizontalBox
   widget=AlertIcon
   properties {brush {
     bIsDynamicallyLoaded=false
     drawAs=Image
     imageType=FullColor
     mirroring=NoMirror
     resourceName=None
     resourceObject=/FortniteCoreUI/UI/Login/Art/IconAlert
     tiling=NoTile
     imageSize=<64 64>
     margin=<ltrb 0 0 0 0>
     outlineSettings {
      bUseBrushTransparency=false
      roundingType=HalfHeightRadius
      width=0
      color {colorUseRule=UseColor_Specified, specifiedColor=<rgba 0 0 0 0>}
      cornerRadii=<0 0 0 0>
     }
     tintColor {colorUseRule=UseColor_Specified, specifiedColor=<rgba 1 1 1 1>}
     uVRegion {bIsValid=false, max=<0 0>, min=<0 0>}
    }}
   slot {class=HorizontalBoxSlot, properties {padding=<ltrb 0 0 15 14>}}
  }
  TitleTextBlock {
   widgetClassPath=&CommonTextBlock
   bInherited=true
   bIsVariable=true
   parent=Title_HorizontalBox
   widget=TitleTextBlock
   properties {
    justification=Center
    text=TITLE
    font {
     bForceMonospaced=false
     bMaterialIsStencil=false
     fontMaterial=null
     fontObject=&HeadingNow
     letterSpacing=0
     monospacedWidth=1
     size=41.25
     skewAmount=0
     typefaceFontName=76 Bold
     outlineSettings {
      bApplyOutlineToDropShadows=false
      bMiteredCorners=false
      bSeparateFillAlpha=false
      outlineMaterial=null
      outlineSize=0
      outlineColor=<rgba 0 0 0 1>
     }
    }
    shadowOffset=<0 0>
   }
   slot {class=HorizontalBoxSlot, properties {padding=<ltrb 0 0 0 14>}}
  }
  Body_ScrollBox {
   widgetClassPath=&ScrollBox
   parent=VerticalBox
   widget=Body_ScrollBox
   slot {class=VerticalBoxSlot, properties {padding=<ltrb 45 45 45 45>, size {sizeRule=Fill, value=1}}}
  }
  Overlay_142 {widgetClassPath=&Overlay, parent=Body_ScrollBox, widget=Overlay_142, slot {class=ScrollBoxSlot}}
  BodyTextBlock {
   widgetClassPath=&CommonTextBlock
   bInherited=true
   bIsVariable=true
   parent=Overlay_142
   widget=BodyTextBlock
   properties {
    autoWrapText=true
    justification=Center
    text=Text Block
    font {
     bForceMonospaced=false
     bMaterialIsStencil=false
     fontMaterial=null
     fontObject=&HeadingNow
     letterSpacing=0
     monospacedWidth=1
     size=35.25
     skewAmount=0
     typefaceFontName=74 Regular
     outlineSettings {
      bApplyOutlineToDropShadows=false
      bMiteredCorners=false
      bSeparateFillAlpha=false
      outlineMaterial=null
      outlineSize=0
      outlineColor=<rgba 0 0 0 1>
     }
    }
    shadowOffset=<0 0>
   }
   slot {class=OverlaySlot, properties {horizontalAlignment=HAlign_Fill, verticalAlignment=VAlign_Fill}}
  }
  ActionButton {
   widgetClassPath=&Button
   bInherited=true
   parent=VerticalBox
   widget=ActionButton
   properties {
    backgroundColor=<rgba 0.9301 1 0.0103 1>
    colorAndOpacity=<rgba 0 0 0 1>
    widgetStyle {
     clickedSlateSound {resourceObject=null}
     disabled {
      bIsDynamicallyLoaded=false
      drawAs=NoDrawType
      imageType=NoImage
      mirroring=NoMirror
      resourceName=None
      resourceObject=null
      tiling=NoTile
      imageSize=<0 0>
      margin=<ltrb 0 0 0 0>
      outlineSettings {
       bUseBrushTransparency=false
       roundingType=HalfHeightRadius
       width=0
       color {colorUseRule=UseColor_Specified, specifiedColor=<rgba 0 0 0 0>}
       cornerRadii=<0 0 0 0>
      }
      tintColor {colorUseRule=UseColor_Specified, specifiedColor=<rgba 1 1 1 1>}
      uVRegion {bIsValid=false, max=<0 0>, min=<0 0>}
     }
     disabledForeground {colorUseRule=UseColor_Specified, specifiedColor=<rgba 0.5271 0.5271 0.5271 1>}
     hovered {
      bIsDynamicallyLoaded=false
      drawAs=RoundedBox
      imageType=NoImage
      mirroring=NoMirror
      resourceName=None
      resourceObject=null
      tiling=NoTile
      imageSize=<32 32>
      margin=<ltrb 0 0 0 0>
      outlineSettings {
       bUseBrushTransparency=true
       roundingType=FixedRadius
       width=0
       color {colorUseRule=UseColor_Specified, specifiedColor=<rgba 0.7243 0.7243 0.7243 1>}
       cornerRadii=<20 20 20 20>
      }
      tintColor {colorUseRule=UseColor_Specified, specifiedColor=<rgba 1 1 1 1>}
      uVRegion {bIsValid=false, max=<0 0>, min=<0 0>}
     }
     hoveredForeground {colorUseRule=UseColor_Specified, specifiedColor=<rgba 0 1 1 1>}
     hoveredSlateSound {resourceObject=null}
     normal {
      bIsDynamicallyLoaded=false
      drawAs=RoundedBox
      imageType=NoImage
      mirroring=NoMirror
      resourceName=None
      resourceObject=null
      tiling=NoTile
      imageSize=<32 32>
      margin=<ltrb 0 0 0 0>
      outlineSettings {
       bUseBrushTransparency=true
       roundingType=FixedRadius
       width=0
       color {colorUseRule=UseColor_Specified, specifiedColor=<rgba 0.6951 0.6951 0.6951 1>}
       cornerRadii=<20 20 20 20>
      }
      tintColor {colorUseRule=UseColor_Specified, specifiedColor=<rgba 1 1 1 1>}
      uVRegion {bIsValid=false, max=<0 0>, min=<0 0>}
     }
     normalForeground {colorUseRule=UseColor_Specified, specifiedColor=<rgba 0 1 1 1>}
     normalPadding=<ltrb 12 1.5 12 1.5>
     pressed {
      bIsDynamicallyLoaded=false
      drawAs=RoundedBox
      imageType=NoImage
      mirroring=NoMirror
      resourceName=None
      resourceObject=null
      tiling=NoTile
      imageSize=<32 32>
      margin=<ltrb 0 0 0 0>
      outlineSettings {
       bUseBrushTransparency=true
       roundingType=FixedRadius
       width=0
       color {colorUseRule=UseColor_Specified, specifiedColor=<rgba 0.7243 0.7243 0.7243 1>}
       cornerRadii=<20 20 20 20>
      }
      tintColor {colorUseRule=UseColor_Specified, specifiedColor=<rgba 0.3843 0.3843 0.3843 1>}
      uVRegion {bIsValid=false, max=<0 0>, min=<0 0>}
     }
     pressedForeground {colorUseRule=UseColor_Specified, specifiedColor=<rgba 0 1 1 1>}
     pressedPadding=<ltrb 12 1.5 12 1.5>
     pressedSlateSound {resourceObject=null}
    }
   }
   slot {class=VerticalBoxSlot, properties {padding=<ltrb 30 20 30 30>}}
  }
  ScaleBox_0 {
   widgetClassPath=&ScaleBox
   parent=ActionButton
   widget=ScaleBox_0
   properties {stretchDirection=DownOnly}
   slot {class=ButtonSlot, properties {padding=<ltrb 15 10 15 10>}}
  }
  ActionTextBlock {
   widgetClassPath=&CommonTextBlock
   bInherited=true
   bIsVariable=true
   parent=ScaleBox_0
   widget=ActionTextBlock
   properties {
    text=ACTION
    font {
     bForceMonospaced=false
     bMaterialIsStencil=false
     fontMaterial=null
     fontObject=&HeadingNow
     letterSpacing=0
     monospacedWidth=1
     size=37.5
     skewAmount=0
     typefaceFontName=76 Bold
     outlineSettings {
      bApplyOutlineToDropShadows=false
      bMiteredCorners=false
      bSeparateFillAlpha=false
      outlineMaterial=null
      outlineSize=0
      outlineColor=<rgba 0 0 0 1>
     }
    }
    shadowOffset=<0 0>
   }
   slot {class=ScaleBoxSlot}
  }
 ]
}
