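# Widget Blueprint for the post-login holding tank's download progress display, built on FortHoldingTankProgressionWidget.
# A bottom-anchored safe-zone layout shows action text with an ellipsis and downloaded bytes in a StackBox, above a styled progress bar and percent label.
# The text blocks and progress bar are inherited bindings, and the graph holds no logic to update them.
# The OnOrientationChanged event switches the StackBox to vertical in portrait and horizontal otherwise.
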WidgetBlueprint /FortniteCoreUI/UI/Login/PostLoginHoldingTank/ProgressBar/HoldingTankProgressionWidget
graphs [EventGraph {nodes [
   n1 = Event OnOrientationChanged!() -> (OutputDelegate, then, IsPortrait:bool) {is=title K2Node_Event}
   > Set Orientation!(
    execute
    Orientation:EOrientation Enum=Select(Option 0:EOrientation Enum=Orient_Horizontal, Option 1:EOrientation Enum=Orient_Vertical, Index=@n1.IsPortrait) -> (ReturnValue:EOrientation Enum)
    self:Stack Box Object Reference=Get StackBox() -> (StackBox:Stack Box Object Reference)
   ) -> (then, Output_Get:EOrientation Enum=Orient_Horizontal)
   Event PreConstruct()
   Event Construct()
   Event Tick()
  ]}]
variables []
widgets {
 info {
  inheritedWidgetCount=5
  parentClass=/Script/FortniteUI.FortHoldingTankProgressionWidget
  rootWidgetClass=&SafeZone
  widgetCount=13
 }
 widgets [
  SafeZone {widgetClassPath=&SafeZone, widget=SafeZone, properties {padLeft=false, padRight=false}}
  Overlay_75 {widgetClassPath=&Overlay, parent=SafeZone, widget=Overlay_75, slot {class=SafeZoneSlot}}
  VerticalBox_109 {
   widgetClassPath=&VerticalBox
   parent=Overlay_75
   widget=VerticalBox_109
   slot {class=OverlaySlot, properties {horizontalAlignment=HAlign_Fill, verticalAlignment=VAlign_Bottom, padding=<ltrb 140 110 140 110>}}
  }
  StackBox {
   widgetClassPath=&StackBox
   bIsVariable=true
   parent=VerticalBox_109
   widget=StackBox
   properties {bOverride_Cursor=true}
   slot {class=VerticalBoxSlot}
  }
  HorizontalBox_DownloadText {
   widgetClassPath=&HorizontalBox
   parent=StackBox
   widget=HorizontalBox_DownloadText
   slot {class=StackBoxSlot, properties {padding=<ltrb 0 0 20 0>}}
  }
  PatchingActionText {
   widgetClassPath=&CommonTextBlock
   bInherited=true
   bIsVariable=true
   parent=HorizontalBox_DownloadText
   widget=PatchingActionText
   properties {
    text=DOWNLOADING
    font {
     bForceMonospaced=false
     bMaterialIsStencil=false
     fontMaterial=null
     fontObject=&HeadingNow
     letterSpacing=0
     monospacedWidth=1
     size=30
     skewAmount=0
     typefaceFontName=86 Bold Italic
     outlineSettings {
      bApplyOutlineToDropShadows=false
      bMiteredCorners=false
      bSeparateFillAlpha=false
      outlineMaterial=null
      outlineSize=0
      outlineColor=<rgba 0 0 0 1>
     }
    }
    shadowOffset=<0 0>
   }
   slot {class=HorizontalBoxSlot, properties {horizontalAlignment=HAlign_Left}}
  }
  UpdateEllipsisText {
   widgetClassPath=&CommonTextBlock
   bInherited=true
   bIsVariable=true
   parent=HorizontalBox_DownloadText
   widget=UpdateEllipsisText
   properties {
    text=...
    font {
     bForceMonospaced=true
     bMaterialIsStencil=false
     fontMaterial=null
     fontObject=&HeadingNow
     letterSpacing=0
     monospacedWidth=0.75
     size=30
     skewAmount=0
     typefaceFontName=86 Bold Italic
     outlineSettings {
      bApplyOutlineToDropShadows=false
      bMiteredCorners=false
      bSeparateFillAlpha=false
      outlineMaterial=null
      outlineSize=0
      outlineColor=<rgba 0 0 0 1>
     }
    }
    shadowOffset=<0 0>
   }
   slot {class=HorizontalBoxSlot, properties {horizontalAlignment=HAlign_Left}}
  }
  DownloadBytesText {
   widgetClassPath=&CommonTextBlock
   bInherited=true
   bIsVariable=true
   parent=StackBox
   widget=DownloadBytesText
   properties {
    text=98.46 MB / 11.63 GB
    font {
     bForceMonospaced=false
     bMaterialIsStencil=false
     fontMaterial=null
     fontObject=&HeadingNow
     letterSpacing=0
     monospacedWidth=1
     size=21
     skewAmount=0
     typefaceFontName=75 Medium
     outlineSettings {
      bApplyOutlineToDropShadows=false
      bMiteredCorners=false
      bSeparateFillAlpha=false
      outlineMaterial=null
      outlineSize=0
      outlineColor=<rgba 0 0 0 1>
     }
    }
    shadowOffset=<0 0>
   }
   slot {class=StackBoxSlot, properties {verticalAlignment=VAlign_Center}}
  }
  HorizontalBox_113 {
   widgetClassPath=&HorizontalBox
   parent=VerticalBox_109
   widget=HorizontalBox_113
   slot {class=VerticalBoxSlot, properties {padding=<ltrb 0 4 0 0>}}
  }
  SizeBox_0 {
   widgetClassPath=&SizeBox
   parent=HorizontalBox_113
   widget=SizeBox_0
   properties {bOverride_HeightOverride=true, heightOverride=22}
   slot {class=HorizontalBoxSlot, properties {size {sizeRule=Fill, value=1}}}
  }
  ProgressBar {
   widgetClassPath=/Script/UMG.ProgressBar
   bInherited=true
   bIsVariable=true
   parent=SizeBox_0
   widget=ProgressBar
   properties {
    percent=0.7
    fillColorAndOpacity=<rgba 0 0.5 1 1>
    widgetStyle {
     enableFillAnimation=false
     backgroundImage {
      bIsDynamicallyLoaded=false
      drawAs=Box
      imageType=FullColor
      mirroring=NoMirror
      resourceName=None
      resourceObject=/FortniteCoreUI/UI/Login/Art/NewProgressBar_Background
      tiling=NoTile
      imageSize=<12 12>
      margin=<ltrb 0.4167 0.4167 0.4167 0.4167>
      outlineSettings {
       bUseBrushTransparency=false
       roundingType=HalfHeightRadius
       width=0
       color {colorUseRule=UseColor_Specified, specifiedColor=<rgba 0 0 0 0>}
       cornerRadii=<0 0 0 0>
      }
      tintColor {colorUseRule=UseColor_Specified, specifiedColor=<rgba 1 1 1 1>}
      uVRegion {bIsValid=false, max=<0 0>, min=<0 0>}
     }
     fillImage {
      bIsDynamicallyLoaded=false
      drawAs=Box
      imageType=FullColor
      mirroring=NoMirror
      resourceName=None
      resourceObject=/FortniteCoreUI/UI/Login/Art/NewProgressBar_Fill
      tiling=NoTile
      imageSize=<12 12>
      margin=<ltrb 0.4167 0.4167 0.4167 0.4167>
      outlineSettings {
       bUseBrushTransparency=false
       roundingType=HalfHeightRadius
       width=0
       color {colorUseRule=UseColor_Specified, specifiedColor=<rgba 0 0 0 0>}
       cornerRadii=<0 0 0 0>
      }
      tintColor {colorUseRule=UseColor_Specified, specifiedColor=<rgba 1 1 1 1>}
      uVRegion {bIsValid=false, max=<0 0>, min=<0 0>}
     }
     marqueeImage {
      bIsDynamicallyLoaded=false
      drawAs=Image
      imageType=FullColor
      mirroring=NoMirror
      resourceName=None
      resourceObject=/FortniteCoreUI/UI/Login/Art/ProgressBar_Marquee
      tiling=Horizontal
      imageSize=<12 12>
      margin=<ltrb 0 0 0 0>
      outlineSettings {
       bUseBrushTransparency=false
       roundingType=HalfHeightRadius
       width=0
       color {colorUseRule=UseColor_Specified, specifiedColor=<rgba 0 0 0 0>}
       cornerRadii=<0 0 0 0>
      }
      tintColor {colorUseRule=UseColor_Specified, specifiedColor=<rgba 1 1 1 1>}
      uVRegion {bIsValid=false, max=<0 0>, min=<0 0>}
     }
    }
   }
   slot {class=SizeBoxSlot, properties {verticalAlignment=VAlign_Center}}
  }
  SizeBox_1 {
   widgetClassPath=&SizeBox
   parent=HorizontalBox_113
   widget=SizeBox_1
   properties {bOverride_WidthOverride=true, widthOverride=100}
   slot {class=HorizontalBoxSlot}
  }
  ProgressionPercentText {
   widgetClassPath=&CommonTextBlock
   bInherited=true
   bIsVariable=true
   parent=SizeBox_1
   widget=ProgressionPercentText
   properties {
    text=76%
    colorAndOpacity {colorUseRule=UseColor_Specified, specifiedColor=<rgba 0 1 1 1>}
    font {
     bForceMonospaced=false
     bMaterialIsStencil=false
     fontMaterial=null
     fontObject=&HeadingNow
     letterSpacing=0
     monospacedWidth=1
     size=19.5
     skewAmount=0
     typefaceFontName=75 Medium
     outlineSettings {
      bApplyOutlineToDropShadows=false
      bMiteredCorners=false
      bSeparateFillAlpha=false
      outlineMaterial=null
      outlineSize=0
      outlineColor=<rgba 0 0 0 1>
     }
    }
    shadowOffset=<0 0>
   }
   slot {class=SizeBoxSlot, properties {horizontalAlignment=HAlign_Right}}
  }
 ]
}
